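# Control Rig for the Rebirth Soldier medium male body that procedurally corrects shoulder straps, collar and clavicle push joints.
# Clavicle swing readers and upper-arm spherical pose readers weight additive bone offsets, adjusted by rigid-body, crouch and targeting states.
# Parent and aim constraints then place shoulder reference bones and nulls, and interpolated emote flags blend final corrections.
# A construction graph spawns helper nulls and FK controls; the shoulder-angle function is not called in the main graph.
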ControlRigBlueprint /Game/Characters/Player/Male/Medium/Bodies/M_MED_Rebirth_Soldier/Meshes/M_MED_Rebirth_Soldier_CtrlRig
graphs [
 Multiply_3_ContainedGraph {nodes [
   n1 = RigVMFunctionEntryNode()
   RigVMFunctionReturnNode(Result=&RigVMFunction_MathDoubleMul(B=@n1.C, A=&RigVMFunction_MathDoubleMul(A=@n1.A, B=@n1.B).Result).Result)
  ]}
 New Function_ContainedGraph {nodes [
   RigVMFunctionEntryNode()
   RigUnit_GetTransform = &RigUnit_GetTransform(Item=(Type=Null,Name="dyn_collar_l_end_null"), bInitial=False)
   n1 = &RigUnit_GetTransform(Item=(Type=Bone,Name="dyn_collar_l_MSTR"), bInitial=False)
   Subtract = /Script/RigVM.RigVMFunction_MathVectorSub(A=@RigUnit_GetTransform.Transform.Translation, B=@n1.Transform.Translation) {is=struct RigVMUnitNode, template="Subtract::Execute(in A,in B,out Result)"}
   RigVMFunction_MathVectorAngle = /Script/RigVM.RigVMFunction_MathVectorAngle(
    A=@Subtract.Result
    B=&RigVMFunction_MathVectorSub(
     A=&RigUnit_GetTransform(Item=(Type=Null,Name="shoulderVectorTarget_l"), bInitial=False).Transform.Translation
     B=@n1.Transform.Translation
    ) {template="Subtract::Execute(in A,in B,out Result)"}.Result
   ) {is=struct RigVMUnitNode, template="AngleBetween::Execute(in A,in B,out Result)"}
   Degrees = /Script/RigVM.RigVMFunction_MathDoubleDeg(Value=@RigVMFunction_MathVectorAngle.Result) {is=struct RigVMUnitNode, template="Degrees::Execute(in Value,out Result)"}
   Remap = &RigVMFunction_MathDoubleRemap(SourceMaximum=37, SourceMinimum=60, TargetMaximum=1, TargetMinimum=0, bClamp=False, Value=@Degrees.Result)
   > &RigUnit_ModifyTransforms(
    ItemToModify=((Item=(Type=Bone,Name="dyn_backstraps_l"),Transform=(Rotation=(X=-0.0016,Y=0.2491,Z=0.0319,W=-0.968),Translation=(X=6.5396,Y=0,Z=0.7309),Scale3D=(X=0.8,Y=1,Z=1))))
    WeightMinimum=0
    Weight=@Remap.Result
   )
   n2 = &RigUnit_GetTransform(Item=(Type=Bone,Name="dyn_collar_r_MSTR"), bInitial=False)
   n3 = &RigVMFunction_MathVectorAngle(
    A=&RigVMFunction_MathVectorSub(
     A=&RigUnit_GetTransform(Item=(Type=Null,Name="dyn_collar_r_end_null"), bInitial=False).Transform.Translation
     B=@n2.Transform.Translation
    ) {template="Subtract::Execute(in A,in B,out Result)"}.Result
    B=&RigVMFunction_MathVectorSub(
     A=&RigUnit_GetTransform(Item=(Type=Null,Name="shoulderVectorTarget_r"), bInitial=False).Transform.Translation
     B=@n2.Transform.Translation
    ) {template="Subtract::Execute(in A,in B,out Result)"}.Result
   ) {template="AngleBetween::Execute(in A,in B,out Result)"}
   > &RigUnit_ModifyTransforms(
    ItemToModify=((Item=(Type=Bone,Name="base_backstraps_r"),Transform=(Rotation=(X=-0.0016,Y=0.2491,Z=0.0319,W=-0.968),Translation=(X=-6.5396,Y=0,Z=-0.7309),Scale3D=(X=0.8,Y=1,Z=1))))
    WeightMinimum=0
    Weight=&RigVMFunction_MathDoubleRemap(
     SourceMaximum=37
     SourceMinimum=60
     TargetMaximum=1
     TargetMinimum=0
     bClamp=False
     Value=&RigVMFunction_MathDoubleDeg(Value=@n3.Result) {template="Degrees::Execute(in Value,out Result)"}.Result
    ).Result
   )
   > RigVMFunctionReturnNode()
  ]}
 New Function_ContainedGraph {nodes [
   n1 = RigVMFunctionEntryNode()
   RigVMFunction_MathBoolToInteger = &RigVMFunction_MathBoolToInteger(Value=@n1.onlyTwist)
   RigVMFunction_MathQuaternionSwingTwist = &RigVMFunction_MathQuaternionSwingTwist(
    TwistAxis=@n1.twistAxis
    Input=&RigUnit_GetRelativeTransformForItem(bChildInitial=False, bParentInitial=False, Child=@n1.child, Parent=@n1.parent).RelativeTransform.Rotation
   )
   n2 = RigVMDispatchNode(
    Values=((),())
    Index=@RigVMFunction_MathBoolToInteger.Result
    Values.0=@RigVMFunction_MathQuaternionSwingTwist.Swing
    Values.1=@RigVMFunction_MathQuaternionSwingTwist.Twist
   ) {template="DISPATCH_RigVMDispatch_SelectInt32(in Index,in Values,out Result)"}
   RigVMFunction_MathQuaternionToEuler = &RigVMFunction_MathQuaternionToEuler(Value=@n2.Result, RotationOrder=@n1.Rotation Order)
   > DISPATCH_RigVMDispatch_Print = RigVMDispatchNode(ScreenColor=(R=1,G=1,B=1,A=1), ScreenDuration=1.5, Enabled=@n1.debugRotation, Value=@RigVMFunction_MathQuaternionToEuler.Result) {template="DISPATCH_RigVMDispatch_Print(in Prefix,in Value,in Enabled,in ScreenDuration,in ScreenColor)"}
   Cast = RigVMDispatchNode(Value=@n1.triggerAxis) {template="DISPATCH_RigVMDispatch_CastEnumToInt(in Value,out Result)"}
   DISPATCH_RigVMDispatch_SelectInt32 = RigVMDispatchNode(
    Values=(0,0,0)
    Index=@Cast.Result
    Values.0=@RigVMFunction_MathQuaternionToEuler.Result.X
    Values.1=@RigVMFunction_MathQuaternionToEuler.Result.Y
    Values.2=@RigVMFunction_MathQuaternionToEuler.Result.Z
   ) {template="DISPATCH_RigVMDispatch_SelectInt32(in Index,in Values,out Result)"}
   Remap = &RigVMFunction_MathDoubleRemap(
    bClamp=False
    Value=@DISPATCH_RigVMDispatch_SelectInt32.Result
    TargetMinimum=@n1.outputMinMax.X
    TargetMaximum=@n1.outputMinMax.Y
    SourceMinimum=@n1.rotationMinMax.X
    SourceMaximum=@n1.rotationMinMax.Y
   )
   Absolute = &RigVMFunction_MathDoubleAbs(Value=@Remap.Result)
   Multiply = &RigVMFunction_MathDoubleMul(
    A=&RigVMFunction_MathDoubleClamp(Value=RigVMDispatchNode(True=@Absolute.Result, Condition=@n1.absoluteValue, False=@Remap.Result) {template="DISPATCH_RigVMDispatch_If(in Condition,in True,in False,out Result)"}.Result).Result
    B=@n1.premult
   )
   > RigVMFunctionReturnNode(outputQuat=@n2.Result, outputWeight=@Multiply.Result)
  ]}
 New Function_ContainedGraph {nodes [
   RigVMFunctionEntryNode()
   > RigVMFunction_Sequence = RigVMAggregateNode()
   &RigUnit_ParentConstraint(
    AdvancedSettings=(RotationOrderForFilter=XZY)
    Child=(Type=Bone,Name="ref_shulderTopPoint_l_base")
    Parents="((Item=(Type=Bone,Name=\"spine_04\"),Weight=0.0277),(Item=(Type=Bone,Name=\"spine_03\"),Weight=0.0464),(Item=(Type=Bone,Name=\"spine_05\"),Weight=0.1001),(Item=(Type=Bone,Name=\"upperarm_twist_01_l\"),Weight=0.3065),(Item=(Type=Bone,Name=\"clavicle_l\"),Weight=0.519 ...(truncated)"
    Weight=1
    ExecutePin=@RigVMFunction_Sequence.A
   )
   > &RigUnit_ParentConstraint(
    AdvancedSettings=(RotationOrderForFilter=XZY)
    Child=(Type=Bone,Name="ref_shulderTopPoint_r_base")
    Parents="((Item=(Type=Bone,Name=\"spine_04\"),Weight=0.0277),(Item=(Type=Bone,Name=\"spine_03\"),Weight=0.0464),(Item=(Type=Bone,Name=\"spine_05\"),Weight=0.1002),(Item=(Type=Bone,Name=\"upperarm_twist_01_r\"),Weight=0.3064),(Item=(Type=Bone,Name=\"clavicle_r\"),Weight=0.519 ...(truncated)"
    Weight=1
   )
   > &RigUnit_AimConstraintLocalSpaceOffset(
    Child=(Type=Bone,Name="ref_shulderTopPoint_l_base")
    Filter=(bX=false,bY=True,bZ=false)
    Parents=((Item=(Type=Bone,Name="dyn_collar_l_MSTR"),Weight=1))
    Weight=1
    WorldUp=(Target=(X=0,Y=0,Z=1),Kind=Location,Space=(Type=None,Name="spine_05"))
   )
   > &RigUnit_AimConstraintLocalSpaceOffset(
    Child=(Type=Bone,Name="ref_shulderTopPoint_r_base")
    Filter=(bX=false,bY=True,bZ=false)
    Parents=((Item=(Type=Bone,Name="dyn_collar_r_MSTR"),Weight=1))
    Weight=1
    WorldUp=(Target=(X=0,Y=0,Z=1),Kind=Location,Space=(Type=None,Name="spine_05"))
   )
   &RigUnit_ParentConstraint(
    AdvancedSettings=(RotationOrderForFilter=XZY)
    Child=(Type=Null,Name="dyn_collar_l_end_null")
    Parents=((Item=(Type=Bone,Name="dyn_collar_fr_l_01"),Weight=0.1852),(Item=(Type=Bone,Name="dyn_collar_bk_l"),Weight=0.2208),(Item=(Type=Bone,Name="dyn_collar_l"),Weight=0.594))
    Weight=1
    ExecutePin=@RigVMFunction_Sequence.B
   )
   > &RigUnit_ParentConstraint(
    AdvancedSettings=(RotationOrderForFilter=XZY)
    Child=(Type=Null,Name="dyn_collar_r_end_null")
    Parents=((Item=(Type=Bone,Name="dyn_collar_bk_r"),Weight=0.1545),(Item=(Type=Bone,Name="dyn_collar_fr_r_01"),Weight=0.249),(Item=(Type=Bone,Name="dyn_collar_r"),Weight=0.5965))
    Weight=1
   )
   > &RigUnit_ParentConstraint(
    AdvancedSettings=(RotationOrderForFilter=XZY)
    Child=(Type=Null,Name="shoulderVectorTarget_l")
    Parents="((Item=(Type=Bone,Name=\"spine_04\"),Weight=0.0195),(Item=(Type=Bone,Name=\"spine_03\"),Weight=0.0463),(Item=(Type=Bone,Name=\"spine_05\"),Weight=0.0978),(Item=(Type=Bone,Name=\"upperarm_twist_01_l\"),Weight=0.303),(Item=(Type=Bone,Name=\"clavicle_l\"),Weight=0.5335 ...(truncated)"
    Weight=1
   )
   > &RigUnit_ParentConstraint(
    AdvancedSettings=(RotationOrderForFilter=XZY)
    Child=(Type=Null,Name="shoulderVectorTarget_r")
    Parents="((Item=(Type=Bone,Name=\"spine_04\"),Weight=0.0195),(Item=(Type=Bone,Name=\"spine_03\"),Weight=0.0463),(Item=(Type=Bone,Name=\"spine_05\"),Weight=0.0978),(Item=(Type=Bone,Name=\"upperarm_twist_01_r\"),Weight=0.303),(Item=(Type=Bone,Name=\"clavicle_r\"),Weight=0.5335 ...(truncated)"
    Weight=1
   )
   RigVMFunctionReturnNode(ExecuteContext=@RigVMFunction_Sequence.C)
  ]}
 New Function_ContainedGraph {nodes [
   RigVMFunctionEntryNode()
   > DeltaRotation v4 = RigVMCollapseNode(
    Rotation Order=ZYX
    child=(Type=Bone,Name="clavicle_l")
    debugRotation=false
    outputMinMax=(X=0,Y=1)
    parent=(Type=Null,Name="clavicle_l_null")
    premult=1
    rotationMinMax=(X=-10,Y=-43)
    triggerAxis=Y
    twistAxis=(X=1,Y=0,Z=0)
   )
   Multiply = &RigVMFunction_MathDoubleMul(B=0.5, A=@"DeltaRotation v4".outputWeight)
   RigVMFunction_MathBoolToInteger = &RigVMFunction_MathBoolToInteger(Value=RigVMVariableNode() {variable=RigidBodiesEnabled}.Value)
   > &RigUnit_ModifyTransforms(
    ItemToModify="((Item=(Type=Bone,Name=\"dyn_backstraps_l\"),Transform=(Rotation=(X=0.0027,Y=0.3762,Z=0.0318,W=-0.926),Translation=(X=6,Y=-0.0319,Z=1.1702),Scale3D=(X=0.8,Y=1,Z=1))),(Item=(Type=Bone,Name=\"dyn_collar_l_MSTR\"),Transform=(Rotation=(X=0,Y=0.3007,Z=0,W=0.9537),T ...(truncated)"
    WeightMinimum=0
    Weight=RigVMDispatchNode(Values=(0,0), Values.1=@Multiply.Result, Index=@RigVMFunction_MathBoolToInteger.Result, Values.0=@"DeltaRotation v4".outputWeight) {template="DISPATCH_RigVMDispatch_SelectInt32(in Index,in Values,out Result)"}.Result
   )
   Remap = &RigVMFunction_MathDoubleRemap(SourceMaximum=1, SourceMinimum=0.1, TargetMaximum=1, TargetMinimum=0, Value=@"DeltaRotation v4".outputWeight)
   > &RigUnit_ModifyTransforms(
    ItemToModify=((Item=(Type=Bone,Name="dyn_clavPushJnt_l"),Transform=(Rotation=(X=0,Y=0,Z=0,W=-1),Translation=(X=5,Y=0,Z=0),Scale3D=(X=1,Y=1,Z=1))))
    WeightMinimum=0
    Weight=@Remap.Result
   )
   n1 = RigVMVariableNode() {variable=isTargetting}
   n2 = &RigVMFunction_AlphaInterp(
    InterpSpeedDecreasing=10
    InterpSpeedIncreasing=5
    Scale=1
    bClampResult=False
    Value=&RigVMFunction_MathBoolToFloat(Value=&RigVMFunction_MathBoolNot(Value=@n1.Value).Result).Result
   )
   RigVMFunction_MathBoolToFloat = &RigVMFunction_MathBoolToFloat(Value=RigVMVariableNode() {variable=RigidBodiesEnabled}.Value)
   Multiply_2 = RigVMAggregateNode(C=@n2.Result, B=@RigVMFunction_MathBoolToFloat.Result, A=@"DeltaRotation v4".outputWeight)
   > &RigUnit_ModifyTransforms(
    ItemToModify="((Item=(Type=Bone,Name=\"dyn_collar_l_MSTR\"),Transform=(Rotation=(X=0,Y=0.5,Z=0,W=0.866),Translation=(X=0,Y=0,Z=0),Scale3D=(X=1,Y=1,Z=1))),(Item=(Type=Bone,Name=\"dyn_collar_fr_l_MSTR\"),Transform=(Rotation=(X=0,Y=0.1736,Z=0,W=0.9848),Translation=(X=2,Y=0,Z=- ...(truncated)"
    WeightMinimum=0
    Weight=@Multiply_2.Result
   )
   > DeltaRotation v4_1 = RigVMCollapseNode(
    Rotation Order=ZYX
    child=(Type=Bone,Name="clavicle_r")
    debugRotation=false
    outputMinMax=(X=0,Y=1)
    parent=(Type=Null,Name="clavicle_r_null")
    premult=1
    rotationMinMax=(X=-14,Y=-43)
    triggerAxis=Y
    twistAxis=(X=1,Y=0,Z=0)
   )
   > &RigUnit_ModifyTransforms(
    ItemToModify="((Item=(Type=Bone,Name=\"dyn_backstraps_r\"),Transform=(Rotation=(X=-0.0089,Y=0.0182,Z=0.0306,W=-0.9993),Translation=(X=-7,Y=0.0172,Z=-1.24),Scale3D=(X=0.8,Y=1,Z=1))),(Item=(Type=Bone,Name=\"dyn_collar_r_MSTR\"),Transform=(Rotation=(X=0,Y=0.2756,Z=0,W=0.9613), ...(truncated)"
    WeightMinimum=0
    Weight=&RigVMFunction_MathDoubleMul(B=RigVMVariableNode() {variable=DyanamicsOnMultiplier}.Value, A=@"DeltaRotation v4_1".outputWeight).Result
   )
   > &RigUnit_ModifyTransforms(
    ItemToModify=((Item=(Type=Bone,Name="dyn_clavPushJnt_r"),Transform=(Rotation=(X=0,Y=0,Z=0,W=-1),Translation=(X=-3.5,Y=0,Z=0),Scale3D=(X=1,Y=1,Z=1))))
    WeightMinimum=0
    Weight=@"DeltaRotation v4_1".outputWeight
   )
   > n3 = &RigUnit_SphericalPoseReader(
    ActiveRegionScaleFactors=(PositiveWidth=1,NegativeWidth=1,PositiveHeight=2,NegativeHeight=1)
    ActiveRegionSize=0.05
    Debug=(bDrawDebug=false,bDraw2D=False,bDrawLocalAxes=False,DebugScale=25,DebugSegments=20,DebugThickness=0.25)
    DriverAxis=(X=-1,Y=0,Z=0)
    DriverItem=(Type=Bone,Name="upperarm_r")
    FalloffRegionScaleFactors=(PositiveWidth=1,NegativeWidth=1,PositiveHeight=1,NegativeHeight=1)
    FalloffSize=0.25
    FlipHeightScaling=False
    FlipWidthScaling=False
    OptionalParentItem=(Type=Bone,Name="None")
    RotationOffset=(X=-80,Y=10,Z=10)
   )
   n4 = RigVMVariableNode() {variable=DyanamicsOnMultiplier}
   n5 = RigVMVariableNode() {variable=isCroching}
   > &RigUnit_ModifyTransforms(
    ItemToModify="((Item=(Type=Bone,Name=\"dyn_collar_fr_r_MSTR\"),Transform=(Rotation=(X=0.0858,Y=-0.173,Z=0.0151,W=0.9811),Translation=(X=0,Y=0,Z=0),Scale3D=(X=1,Y=1,Z=1))),(Item=(Type=Bone,Name=\"dyn_collar_r_MSTR\"),Transform=(Rotation=(X=0,Y=0.0872,Z=0,W=0.9962),Translatio ...(truncated)"
    WeightMinimum=0
    Weight=RigVMDispatchNode(
     Values=(0,1)
     Values.0=&RigVMFunction_MathDoubleMul(B=@n4.Value, A=@n3.OutputParam).Result
     Index=&RigVMFunction_MathBoolToInteger(Value=@n5.Value).Result
     Values.1=@n3.OutputParam
    ) {template="DISPATCH_RigVMDispatch_SelectInt32(in Index,in Values,out Result)"}.Result
   )
   > n6 = &RigUnit_SphericalPoseReader(
    ActiveRegionScaleFactors=(PositiveWidth=1,NegativeWidth=1,PositiveHeight=2,NegativeHeight=1)
    ActiveRegionSize=0.05
    Debug=(bDrawDebug=false,bDraw2D=False,bDrawLocalAxes=False,DebugScale=25,DebugSegments=20,DebugThickness=0.25)
    DriverAxis=(X=-1,Y=0,Z=0)
    DriverItem=(Type=Bone,Name="upperarm_r")
    FalloffRegionScaleFactors=(PositiveWidth=1,NegativeWidth=1,PositiveHeight=1,NegativeHeight=1)
    FalloffSize=0.25
    FlipHeightScaling=False
    FlipWidthScaling=False
    OptionalParentItem=(Type=Bone,Name="None")
    RotationOffset=(X=-120,Y=20,Z=20)
   )
   n7 = RigVMVariableNode() {variable=DyanamicsOnMultiplier}
   n8 = RigVMVariableNode() {variable=isCroching}
   > &RigUnit_ModifyTransforms(
    ItemToModify="((Item=(Type=Bone,Name=\"dyn_collar_fr_r_MSTR\"),Transform=(Rotation=(X=0.1727,Y=-0.1029,Z=0.0182,W=0.9794),Translation=(X=0,Y=0,Z=0),Scale3D=(X=1,Y=1,Z=1))),(Item=(Type=Bone,Name=\"dyn_collar_r_MSTR\"),Transform=(Rotation=(X=0,Y=0,Z=0,W=1),Translation=(X=0,Y= ...(truncated)"
    WeightMinimum=0
    Weight=RigVMDispatchNode(
     Values=(0,1)
     Values.0=&RigVMFunction_MathDoubleMul(B=@n7.Value, A=@n6.OutputParam).Result
     Index=&RigVMFunction_MathBoolToInteger(Value=@n8.Value).Result
     Values.1=@n6.OutputParam
    ) {template="DISPATCH_RigVMDispatch_SelectInt32(in Index,in Values,out Result)"}.Result
   )
   RigVMFunction_MathBoolNot = &RigVMFunction_MathBoolNot(Value=RigVMVariableNode() {variable=isCroching}.Value)
   RigVMFunction_MathBoolAnd = &RigVMFunction_MathBoolAnd(B=@RigVMFunction_MathBoolNot.Result, A=RigVMVariableNode() {variable=RigidBodiesEnabled}.Value)
   AlphaInterp = &RigVMFunction_AlphaInterp(
    InterpSpeedDecreasing=10
    InterpSpeedIncreasing=10
    Scale=1
    bClampResult=False
    Value=&RigVMFunction_MathBoolToFloat(Value=@RigVMFunction_MathBoolAnd.Result).Result
   )
   > &RigUnit_ModifyTransforms(
    ItemToModify="((Item=(Type=Bone,Name=\"dyn_collar_r_MSTR\"),Transform=(Rotation=(X=0.173,Y=-0.0858,Z=0.0151,W=0.9811),Translation=(X=3,Y=-2,Z=-2),Scale3D=(X=1,Y=1,Z=1))),(Item=(Type=Bone,Name=\"dyn_collar_fr_r_MSTR\"),Transform=(Rotation=(X=0,Y=0,Z=0,W=1),Translation=(X=1,Y ...(truncated)"
    WeightMinimum=0
    Weight=&RigVMFunction_MathDoubleMul(B=@AlphaInterp.Result, A=@n6.OutputParam).Result
   )
   n9 = RigVMVariableNode() {variable=RigidBodiesEnabled}
   n10 = RigVMVariableNode() {variable=isCroching}
   n11 = &RigVMFunction_AlphaInterp(
    InterpSpeedDecreasing=10
    InterpSpeedIncreasing=10
    Scale=1
    bClampResult=False
    Value=&RigVMFunction_MathBoolToFloat(Value=&RigVMFunction_MathBoolAnd(A=@n9.Value, B=@n10.Value).Result).Result
   )
   > &RigUnit_ModifyTransforms(
    ItemToModify="((Item=(Type=Bone,Name=\"dyn_collar_r_MSTR\"),Transform=(Rotation=(X=0,Y=0.0654,Z=0,W=0.9979),Translation=(X=3,Y=0,Z=2),Scale3D=(X=1,Y=1,Z=1))),(Item=(Type=Bone,Name=\"dyn_collar_fr_r_MSTR\"),Transform=(Rotation=(X=0.1305,Y=0,Z=0,W=0.9914),Translation=(X=1,Y=0 ...(truncated)"
    WeightMinimum=0
    Weight=&RigVMFunction_MathDoubleMul(B=@n11.Result, A=@n6.OutputParam).Result
   )
   > n12 = &RigUnit_SphericalPoseReader(
    ActiveRegionScaleFactors=(PositiveWidth=1,NegativeWidth=1,PositiveHeight=2,NegativeHeight=1)
    ActiveRegionSize=0.05
    Debug=(bDrawDebug=false,bDraw2D=False,bDrawLocalAxes=False,DebugScale=25,DebugSegments=20,DebugThickness=0.25)
    DriverAxis=(X=1,Y=0,Z=0)
    DriverItem=(Type=Bone,Name="upperarm_l")
    FalloffRegionScaleFactors=(PositiveWidth=1,NegativeWidth=1,PositiveHeight=1,NegativeHeight=1)
    FalloffSize=0.25
    FlipHeightScaling=False
    FlipWidthScaling=False
    OptionalParentItem=(Type=Bone,Name="None")
    RotationOffset=(X=85,Y=20,Z=-20)
   )
   > &RigUnit_ModifyTransforms(
    ItemToModify="((Item=(Type=Bone,Name=\"dyn_collar_fr_l_MSTR\"),Transform=(Rotation=(X=0,Y=0.1045,Z=0,W=0.9945),Translation=(X=0,Y=0,Z=0),Scale3D=(X=1,Y=1,Z=1))),(Item=(Type=Bone,Name=\"dyn_collar_l_MSTR\"),Transform=(Rotation=(X=0,Y=0.0785,Z=0,W=0.9969),Translation=(X=0,Y=0 ...(truncated)"
    WeightMinimum=0
    Weight=&RigVMFunction_MathDoubleMul(A=@n12.OutputParam, B=RigVMVariableNode() {variable=DyanamicsOnMultiplier}.Value).Result
   )
   > n13 = &RigUnit_SphericalPoseReader(
    ActiveRegionScaleFactors=(PositiveWidth=1,NegativeWidth=1,PositiveHeight=2,NegativeHeight=1)
    ActiveRegionSize=0.05
    Debug=(bDrawDebug=false,bDraw2D=False,bDrawLocalAxes=False,DebugScale=25,DebugSegments=20,DebugThickness=0.25)
    DriverAxis=(X=1,Y=0,Z=0)
    DriverItem=(Type=Bone,Name="upperarm_l")
    FalloffRegionScaleFactors=(PositiveWidth=1,NegativeWidth=1,PositiveHeight=1,NegativeHeight=1)
    FalloffSize=0.25
    FlipHeightScaling=False
    FlipWidthScaling=False
    OptionalParentItem=(Type=Bone,Name="None")
    RotationOffset=(X=55,Y=20,Z=-25)
   )
   > &RigUnit_ModifyTransforms(
    ItemToModify="((Item=(Type=Bone,Name=\"dyn_collar_fr_l_MSTR\"),Transform=(Rotation=(X=-0.1727,Y=0.1029,Z=0.0182,W=0.9794),Translation=(X=0,Y=0,Z=0),Scale3D=(X=1,Y=1,Z=1))),(Item=(Type=Bone,Name=\"dyn_collar_l_MSTR\"),Transform=(Rotation=(X=0,Y=-0.2588,Z=0,W=0.9659),Translat ...(truncated)"
    WeightMinimum=0
    Weight=&RigVMFunction_MathDoubleMul(A=@n13.OutputParam, B=RigVMVariableNode() {variable=DyanamicsOnMultiplier}.Value).Result
   )
   > n14 = &RigUnit_SphericalPoseReader(
    ActiveRegionScaleFactors=(PositiveWidth=1,NegativeWidth=1,PositiveHeight=2,NegativeHeight=1)
    ActiveRegionSize=0.05
    Debug=(bDrawDebug=false,bDraw2D=false,bDrawLocalAxes=False,DebugScale=25,DebugSegments=20,DebugThickness=0.25)
    DriverAxis=(X=1,Y=0,Z=0)
    DriverItem=(Type=Bone,Name="upperarm_l")
    FalloffRegionScaleFactors=(PositiveWidth=1,NegativeWidth=1,PositiveHeight=1,NegativeHeight=1)
    FalloffSize=0.25
    FlipHeightScaling=False
    FlipWidthScaling=False
    OptionalParentItem=(Type=Bone,Name="None")
    RotationOffset=(X=30,Y=-20,Z=-20)
   )
   n15 = &RigVMFunction_MathBoolNot(Value=RigVMVariableNode() {variable=isCroching}.Value)
   n16 = RigVMVariableNode() {variable=RigidBodiesEnabled}
   n17 = &RigVMFunction_AlphaInterp(
    InterpSpeedDecreasing=10
    InterpSpeedIncreasing=10
    Scale=1
    bClampResult=False
    Value=&RigVMFunction_MathBoolToFloat(Value=&RigVMFunction_MathBoolAnd(B=@n15.Result, A=@n16.Value).Result).Result
   )
   > &RigUnit_ModifyTransforms(
    ItemToModify="((Item=(Type=Bone,Name=\"dyn_collar_fr_l_MSTR\"),Transform=(Rotation=(X=0,Y=0,Z=0,W=1),Translation=(X=0,Y=0,Z=0),Scale3D=(X=1,Y=1,Z=1))),(Item=(Type=Bone,Name=\"dyn_collar_l_MSTR\"),Transform=(Rotation=(X=0,Y=-0.2588,Z=0,W=0.9659),Translation=(X=1,Y=8,Z=-1),Sc ...(truncated)"
    WeightMinimum=0
    Weight=&RigVMFunction_MathDoubleMul(B=@n17.Result, A=@n14.OutputParam).Result
   )
   > RigVMFunctionReturnNode()
  ]}
 New Function_ContainedGraph {nodes [
   n1 = RigVMFunctionEntryNode()
   RigVMFunction_MathBoolToInteger = &RigVMFunction_MathBoolToInteger(Value=@n1.onlyTwist)
   RigVMFunction_MathQuaternionSwingTwist = &RigVMFunction_MathQuaternionSwingTwist(
    TwistAxis=@n1.twistAxis
    Input=&RigUnit_GetRelativeTransformForItem(bChildInitial=False, bParentInitial=False, Child=@n1.child, Parent=@n1.parent).RelativeTransform.Rotation
   )
   n2 = RigVMDispatchNode(
    Values=((),())
    Index=@RigVMFunction_MathBoolToInteger.Result
    Values.0=@RigVMFunction_MathQuaternionSwingTwist.Swing
    Values.1=@RigVMFunction_MathQuaternionSwingTwist.Twist
   ) {template="DISPATCH_RigVMDispatch_SelectInt32(in Index,in Values,out Result)"}
   RigVMFunction_MathQuaternionToEuler = &RigVMFunction_MathQuaternionToEuler(Value=@n2.Result, RotationOrder=@n1.Rotation Order)
   > DISPATCH_RigVMDispatch_Print = RigVMDispatchNode(ScreenColor=(R=1,G=1,B=1,A=1), ScreenDuration=1.5, Enabled=@n1.debugRotation, Value=@RigVMFunction_MathQuaternionToEuler.Result) {template="DISPATCH_RigVMDispatch_Print(in Prefix,in Value,in Enabled,in ScreenDuration,in ScreenColor)"}
   Cast = RigVMDispatchNode(Value=@n1.triggerAxis) {template="DISPATCH_RigVMDispatch_CastEnumToInt(in Value,out Result)"}
   DISPATCH_RigVMDispatch_SelectInt32 = RigVMDispatchNode(
    Values=(0,0,0)
    Index=@Cast.Result
    Values.0=@RigVMFunction_MathQuaternionToEuler.Result.X
    Values.1=@RigVMFunction_MathQuaternionToEuler.Result.Y
    Values.2=@RigVMFunction_MathQuaternionToEuler.Result.Z
   ) {template="DISPATCH_RigVMDispatch_SelectInt32(in Index,in Values,out Result)"}
   Remap = &RigVMFunction_MathDoubleRemap(
    bClamp=False
    Value=@DISPATCH_RigVMDispatch_SelectInt32.Result
    TargetMinimum=@n1.outputMinMax.X
    TargetMaximum=@n1.outputMinMax.Y
    SourceMinimum=@n1.rotationMinMax.X
    SourceMaximum=@n1.rotationMinMax.Y
   )
   Absolute = &RigVMFunction_MathDoubleAbs(Value=@Remap.Result)
   Multiply = &RigVMFunction_MathDoubleMul(
    A=&RigVMFunction_MathDoubleClamp(Value=RigVMDispatchNode(True=@Absolute.Result, Condition=@n1.absoluteValue, False=@Remap.Result) {template="DISPATCH_RigVMDispatch_If(in Condition,in True,in False,out Result)"}.Result).Result
    B=@n1.premult
   )
   > RigVMFunctionReturnNode(outputQuat=@n2.Result, outputWeight=@Multiply.Result)
  ]}
 New Function_ContainedGraph {nodes [
   n1 = RigVMFunctionEntryNode()
   RigVMFunction_MathBoolToInteger = &RigVMFunction_MathBoolToInteger(Value=@n1.onlyTwist)
   RigVMFunction_MathQuaternionSwingTwist = &RigVMFunction_MathQuaternionSwingTwist(
    TwistAxis=@n1.twistAxis
    Input=&RigUnit_GetRelativeTransformForItem(bChildInitial=False, bParentInitial=False, Child=@n1.child, Parent=@n1.parent).RelativeTransform.Rotation
   )
   n2 = RigVMDispatchNode(
    Values=((),())
    Index=@RigVMFunction_MathBoolToInteger.Result
    Values.0=@RigVMFunction_MathQuaternionSwingTwist.Swing
    Values.1=@RigVMFunction_MathQuaternionSwingTwist.Twist
   ) {template="DISPATCH_RigVMDispatch_SelectInt32(in Index,in Values,out Result)"}
   RigVMFunction_MathQuaternionToEuler = &RigVMFunction_MathQuaternionToEuler(Value=@n2.Result, RotationOrder=@n1.Rotation Order)
   > DISPATCH_RigVMDispatch_Print = RigVMDispatchNode(ScreenColor=(R=1,G=1,B=1,A=1), ScreenDuration=1.5, Enabled=@n1.debugRotation, Value=@RigVMFunction_MathQuaternionToEuler.Result) {template="DISPATCH_RigVMDispatch_Print(in Prefix,in Value,in Enabled,in ScreenDuration,in ScreenColor)"}
   Cast = RigVMDispatchNode(Value=@n1.triggerAxis) {template="DISPATCH_RigVMDispatch_CastEnumToInt(in Value,out Result)"}
   DISPATCH_RigVMDispatch_SelectInt32 = RigVMDispatchNode(
    Values=(0,0,0)
    Index=@Cast.Result
    Values.0=@RigVMFunction_MathQuaternionToEuler.Result.X
    Values.1=@RigVMFunction_MathQuaternionToEuler.Result.Y
    Values.2=@RigVMFunction_MathQuaternionToEuler.Result.Z
   ) {template="DISPATCH_RigVMDispatch_SelectInt32(in Index,in Values,out Result)"}
   Remap = &RigVMFunction_MathDoubleRemap(
    bClamp=False
    Value=@DISPATCH_RigVMDispatch_SelectInt32.Result
    TargetMinimum=@n1.outputMinMax.X
    TargetMaximum=@n1.outputMinMax.Y
    SourceMinimum=@n1.rotationMinMax.X
    SourceMaximum=@n1.rotationMinMax.Y
   )
   Absolute = &RigVMFunction_MathDoubleAbs(Value=@Remap.Result)
   Multiply = &RigVMFunction_MathDoubleMul(
    A=&RigVMFunction_MathDoubleClamp(Value=RigVMDispatchNode(True=@Absolute.Result, Condition=@n1.absoluteValue, False=@Remap.Result) {template="DISPATCH_RigVMDispatch_If(in Condition,in True,in False,out Result)"}.Result).Result
    B=@n1.premult
   )
   > RigVMFunctionReturnNode(outputQuat=@n2.Result, outputWeight=@Multiply.Result)
  ]}
 New Function_ContainedGraph {nodes [
   RigVMFunctionEntryNode()
   AlphaInterp = &RigVMFunction_AlphaInterp(
    InterpSpeedDecreasing=5
    InterpSpeedIncreasing=5
    Scale=1
    bClampResult=false
    Value=&RigVMFunction_MathBoolToFloat(Value=RigVMVariableNode() {variable=isEmoting}.Value).Result
   )
   > &RigUnit_ModifyTransforms(
    ItemToModify="((Item=(Type=Bone,Name=\"base_backstraps_l\"),Transform=(Rotation=(X=0,Y=0,Z=0,W=1),Translation=(X=1,Y=0,Z=1.5),Scale3D=(X=0.9,Y=1,Z=1))),(Item=(Type=Bone,Name=\"base_backstraps_r\"),Transform=(Rotation=(X=0,Y=0,Z=0,W=1),Translation=(X=-1,Y=0,Z=-1.5),Scale3D=( ...(truncated)"
    WeightMinimum=0
    Weight=@AlphaInterp.Result
   )
   n1 = &RigVMFunction_AlphaInterp(
    InterpSpeedDecreasing=5
    InterpSpeedIncreasing=5
    Scale=1
    bClampResult=false
    Value=&RigVMFunction_MathBoolToFloat(Value=RigVMVariableNode() {variable=isEmoting}.Value).Result
   )
   > &RigUnit_ModifyTransforms(
    ItemToModify="((Item=(Type=Bone,Name=\"base_backstraps_l\"),Transform=(Rotation=(X=0,Y=0,Z=0,W=1),Translation=(X=0,Y=0,Z=0),Scale3D=(X=1,Y=1,Z=1))),(Item=(Type=Bone,Name=\"dyn_collar_fr_l_MSTR\"),Transform=(Rotation=(X=0,Y=0,Z=0,W=1),Translation=(X=1,Y=0,Z=-1.5),Scale3D=(X= ...(truncated)"
    WeightMinimum=0
    Weight=@n1.Result
   )
   n2 = &RigVMFunction_AlphaInterp(
    InterpSpeedDecreasing=5
    InterpSpeedIncreasing=5
    Scale=1
    bClampResult=false
    Value=&RigVMFunction_MathBoolToFloat(Value=RigVMVariableNode() {variable=apologyEmote}.Value).Result
   )
   > &RigUnit_ModifyTransforms(
    ItemToModify="((Item=(Type=Bone,Name=\"base_backstraps_l\"),Transform=(Rotation=(X=0,Y=0,Z=-0.2588,W=0.9659),Translation=(X=1.5,Y=0,Z=1.5),Scale3D=(X=1,Y=1,Z=1))),(Item=(Type=Bone,Name=\"base_backstraps_r\"),Transform=(Rotation=(X=0,Y=0,Z=0.0872,W=0.9962),Translation=(X=-1. ...(truncated)"
    WeightMinimum=0
    Weight=@n2.Result
   )
   > RigVMFunctionReturnNode()
  ]}
 RigVMFunctionLibrary {nodes [
   Constraints = RigVMCollapseNode(ExecuteContext=())
   DeltaRotation v4 = RigVMCollapseNode(
    ExecuteContext=()
    Rotation Order=ZYX
    child=(Type=Bone,Name="None")
    outputMinMax=(X=0,Y=1)
    parent=(Type=Bone,Name="None")
    premult=1
    rotationMinMax=(X=0,Y=0)
    twistAxis=(X=1,Y=0,Z=0)
   )
   EmoteCorrections = RigVMCollapseNode(ExecuteContext=())
   UpperTorsoAdjustmentsSetup = RigVMCollapseNode(ExecuteContext=())
   degreeDetectionSetup = RigVMCollapseNode(ExecuteContext=())
  ]}
 RigVMFunction_Sequence_1_ContainedGraph {nodes [
   n1 = RigVMFunctionEntryNode()
   RigVMFunction_Sequence = &RigVMFunction_Sequence(ExecuteContext=@n1.ExecuteContext)
   n2 = &RigVMFunction_Sequence(ExecuteContext=@RigVMFunction_Sequence.B)
   RigVMFunctionReturnNode(A=@RigVMFunction_Sequence.A, B=@n2.A, C=@n2.B)
  ]}
 RigVMModel {nodes [
   RigUnit_BeginExecution = &RigUnit_BeginExecution()
   RigVMFunction_MathBoolToInteger = &RigVMFunction_MathBoolToInteger(Value=RigVMVariableNode() {variable=RigidBodiesEnabled}.Value)
   RigVMVariableNode(
    ExecuteContext=@RigUnit_BeginExecution.ExecutePin
    Value=RigVMDispatchNode(Values=(1,0.5), Index=@RigVMFunction_MathBoolToInteger.Result) {template="DISPATCH_RigVMDispatch_SelectInt32(in Index,in Values,out Result)"}.Result
   ) {variable=DyanamicsOnMultiplier}
   > &RigUnit_ModifyTransforms(
    ItemToModify="((Item=(Type=Bone,Name=\"base_backstraps_l\"),Transform=(Rotation=(X=0,Y=0,Z=0,W=1),Translation=(X=1.5,Y=0,Z=0),Scale3D=(X=1,Y=1,Z=1))),(Item=(Type=Bone,Name=\"base_backstraps_r\"),Transform=(Rotation=(X=0,Y=0,Z=0,W=1),Translation=(X=-1.5,Y=0,Z=0),Scale3D=(X=1 ...(truncated)"
    Weight=1
    WeightMinimum=0
   )
   > UpperTorsoAdjustmentsSetup = RigVMFunctionReferenceNode()
   > Constraints = RigVMFunctionReferenceNode()
   > EmoteCorrections = RigVMFunctionReferenceNode()
   &RigUnit_ModifyTransforms(
    ExecutePin=()
    ItemToModify="((Item=(Type=Bone,Name=\"base_backstraps_l\"),Transform=(Rotation=(X=0,Y=0,Z=-0.2588,W=0.9659),Translation=(X=2,Y=0,Z=2),Scale3D=(X=1,Y=1,Z=1))),(Item=(Type=Bone,Name=\"base_backstraps_r\"),Transform=(Rotation=(X=0,Y=0,Z=-0.2588,W=0.9659),Translation=(X=-2,Y=0 ...(truncated)"
    Weight=1
    WeightMinimum=0
   )
   setFKCtrl = RigVMFunctionReferenceNode(ExecuteContext=(), setFK_Ctrl=false)
  ]}
 RigVMModel Build {nodes [
   n1 = &RigUnit_PrepareForExecution()
   &RigUnit_HierarchyAddNull(
    Name=shoulderVectorTarget_l
    Parent=(Type=Bone,Name="None")
    Space=LocalSpace
    Transform=(Rotation=(X=0,Y=0,Z=0,W=1),Translation=(X=17.7819,Y=-4.8445,Z=141),Scale3D=(X=1,Y=1,Z=1))
    ExecutePin=@n1.ExecutePin
   )
   > &RigUnit_HierarchyAddNull(
    Name=shoulderVectorTarget_r
    Parent=(Type=Bone,Name="None")
    Space=LocalSpace
    Transform=(Rotation=(X=0,Y=0,Z=0,W=1),Translation=(X=-17.7819,Y=-4.8445,Z=141),Scale3D=(X=1,Y=1,Z=1))
   )
   > &RigUnit_HierarchyAddNull(
    Name=dyn_collar_l_end_null
    Parent=(Type=Bone,Name="None")
    Space=LocalSpace
    Transform=(Rotation=(X=0,Y=0,Z=0,W=1),Translation=(X=12,Y=-1.0608,Z=147),Scale3D=(X=1,Y=1,Z=1))
   )
   > &RigUnit_HierarchyAddNull(
    Name=dyn_collar_r_end_null
    Parent=(Type=Bone,Name="None")
    Space=LocalSpace
    Transform=(Rotation=(X=0,Y=0,Z=0,W=1),Translation=(X=-11.5,Y=-1.0608,Z=147),Scale3D=(X=1,Y=1,Z=1))
   )
   > &RigUnit_HierarchyAddNull(
    Name=R_ShoulderTop_
    Parent=(Type=Bone,Name="spine_05")
    Space=LocalSpace
    Transform=(Rotation=(X=0,Y=0,Z=0,W=1),Translation=(X=40,Y=0,Z=0),Scale3D=(X=1,Y=1,Z=1))
   )
   addFKCtrl = RigVMFunctionReferenceNode(
    Custom_FKCtrl="((joint_names_14_#1=((Type=Bone,Name=\"dyn_collar_l_MSTR\")),parent_2_#2=(Type=Control,Name=\"None\"),mirror_8_#3=true,shape_settings_24_#4= ...(truncated)"
    ExecuteContext=()
    Options_FKCtrl="(pelvis_1_#5=true,spine_3_#6=true,clavicle_5_#7=true,arm_7_#8=true,leg_9_#9=true,head_11_46DC15EF465CD2979 ...(truncated)"
    Shape_Settings="(pelvis_2_#10=(bVisible=True,Name=\"Hexagon_Thin\",Color=(R=0.0117,G=0.8281,B=1,A=1),Transform=(Rotation=(X=0,Y=-0.7071,Z=0,W=0.7071),Translation=(X=2,Y=0,Z=0),Scale3D=(X=5,Y=5,Z=7))),spine_4_#11=(bVi ...(truncated)"
   )
  ]}
]
hierarchy [
 root {loc=<0 0 0>, rot=<0 0 0 1>}
 attach {parent=root, loc=<0 0 -1.3227>, rot=<0 0 0.038 -0.9993>}
 pelvis {parent=root, loc=<0 -0.7071 93.1191>, rot=<0.0127 0.707 -0.0127 -0.707>}
 spine_01 {parent=pelvis, loc=<6.7494 -0.6388 0>, rot=<0 0 0.034 -0.9994>}
 spine_02 {parent=spine_01, loc=<7.2518 -0.035 0>, rot=<0 0 0.0238 -0.9997>}
 spine_03 {parent=spine_02, loc=<6.0104 0 0>, rot=<0 0 0.0149 -0.9999>}
 spine_04 {parent=spine_03, loc=<8.1024 0 0>, rot=<0 0 -0.0132 -0.9999>}
 spine_05 {parent=spine_04, loc=<10.0146 0 0>, rot=<0 0 -0.0015 -1>}
 clavicle_l {parent=spine_05, loc=<4.3575 -0.1466 -2.3223>, rot=<0.0465 -0.7289 -0.009 -0.6829>}
 upperarm_l {parent=clavicle_l, loc=<15.5111 0 0>, rot=<0.0226 -0.3234 -0.0088 -0.9459>}
 lowerarm_l {parent=upperarm_l, loc=<26.6514 0 0>, rot=<0.0391 -0.0099 -0.2453 -0.9686>}
 hand_l {parent=lowerarm_l, loc=<22.3491 0 0>, rot=<-0.7081 -0.0507 0.0975 -0.6975>}
 index_metacarpal_l {parent=hand_l, loc=<1.6535 0.0183 -2.96>, rot=<0.0158 -0.0124 -0.0173 -0.9996>}
 index_01_l {parent=index_metacarpal_l, loc=<8.3008 0.0357 -0.0524>, rot=<-0.0044 0.0485 0.0924 -0.9945>}
 index_02_l {parent=index_01_l, loc=<3.3587 -0.3004 0.0452>, rot=<-0.002 0.0027 0.3225 -0.9466>}
 index_03_l {parent=index_02_l, loc=<2.5578 -0.123 -0.0252>, rot=<-0.002 0.0067 0.2558 -0.9667>}
 middle_metacarpal_l {parent=hand_l, loc=<2.1856 -0.1725 -0.5984>, rot=<0.006 0.0168 -0.0329 -0.9993>}
 middle_01_l {parent=middle_metacarpal_l, loc=<7.3275 -0.2447 0.0129>, rot=<-0.0208 0.0042 0.0858 -0.9961>}
 middle_02_l {parent=middle_01_l, loc=<4.0123 0.1401 0.1175>, rot=<-0.0218 0.0258 0.3512 -0.9357>}
 middle_03_l {parent=middle_02_l, loc=<2.6077 0.0534 0.1373>, rot=<0.0082 0.0188 0.1708 -0.9851>}
 pinky_metacarpal_l {parent=hand_l, loc=<1.7616 -0.9753 3.3814>, rot=<-0.0397 0.0814 -0.069 -0.9935>}
 pinky_01_l {parent=pinky_metacarpal_l, loc=<7.1709 -0.4831 -0.1094>, rot=<-0.0654 -0.0035 0.2242 -0.9723>}
 pinky_02_l {parent=pinky_01_l, loc=<2.8274 0.0893 -0.1043>, rot=<0.0037 0.0011 0.2854 -0.9584>}
 pinky_03_l {parent=pinky_02_l, loc=<2.2046 0.0307 -0.0078>, rot=<-0.0002 0.0098 0.1445 -0.9895>}
 ring_metacarpal_l {parent=hand_l, loc=<2.2074 -0.3625 1.3638>, rot=<-0.0284 0.0722 -0.0458 -0.9959>}
 ring_01_l {parent=ring_metacarpal_l, loc=<7.0295 -0.6197 -0.1446>, rot=<-0.0489 -0.0252 0.1892 -0.9804>}
 ring_02_l {parent=ring_01_l, loc=<3.6174 0.4647 -0.1027>, rot=<0 0 0.2806 -0.9598>}
 ring_03_l {parent=ring_02_l, loc=<2.4152 0.114 0.0199>, rot=<0 0 0.1189 -0.9929>}
 thumb_01_l {parent=hand_l, loc=<1.3118 -0.3191 -2.459>, rot=<0.6546 -0.4042 0.0664 -0.6354>}
 thumb_02_l {parent=thumb_01_l, loc=<5.9336 -0.0448 0.0401>, rot=<-0.0037 0.0213 0.298 -0.9543>}
 thumb_03_l {parent=thumb_02_l, loc=<2.2223 0.3682 -0.0907>, rot=<0.0118 -0.013 0.0066 -0.9998>}
 weapon_l {parent=hand_l, loc=<-0.1997 -3.3902 0.3669>, rot=<-0.0006 0.0002 0.0002 -1>}
 lowerarm_twist_01_l {parent=lowerarm_l, loc=<17.8549 0 0>, rot=<0 0 0 1>}
 lowerarm_twist_02_l {parent=lowerarm_l, loc=<9.6199 0 0>, rot=<0 0 0 1>}
 upperarm_twist_01_l {parent=upperarm_l, loc=<0.0041 0 0>, rot=<0 0 0 1>}
 upperarm_twist_02_l {parent=upperarm_l, loc=<13.7941 0 0>, rot=<0 0 0 1>}
 dyn_strap_clip_fr_l_MSTR {parent=clavicle_l, loc=<12.1796 11.1367 -2.2174>, rot=<0.4186 -0.504 -0.3818 -0.6519>}
 dyn_strap_clip_fr_l {parent=dyn_strap_clip_fr_l_MSTR, loc=<0.2625 -0.0438 0.2723>, rot=<-0.0223 0.2541 -0.0075 -0.9669>}
 dyn_strap_clip_bk_l_MSTR {parent=clavicle_l, loc=<11.6531 -11.3141 -0.439>, rot=<-0.4008 -0.4822 0.4921 -0.6038>}
 dyn_strap_clip_bk_l {parent=dyn_strap_clip_bk_l_MSTR, loc=<0.1948 -0.0817 0.3487>, rot=<-0.024 0.1358 -0.0033 -0.9904>}
 base_backstraps_l {parent=clavicle_l, loc=<10.5973 0 8.0767>, rot=<0 -0.0993 0 -0.9951>}
 dyn_backstraps_l {parent=base_backstraps_l, loc=<0 0 0>, rot=<0 0 0 1>}
 base_clavPushJnt_l {parent=clavicle_l, loc=<6.8332 1.9955 7.5743>, rot=<0 -0.0993 0 -0.9951>}
 dyn_clavPushJnt_l {parent=base_clavPushJnt_l, loc=<0 0 0>, rot=<0 0 0 1>}
 clavicle_r {parent=spine_05, loc=<4.3575 -0.1466 2.3223>, rot=<0.7289 0.0465 -0.6829 0.009>}
 upperarm_r {parent=clavicle_r, loc=<-15.5111 0 0>, rot=<0.0226 -0.3234 -0.0088 -0.9459>}
 lowerarm_r {parent=upperarm_r, loc=<-26.6515 0 0>, rot=<0.0391 -0.0099 -0.2453 -0.9686>}
 hand_r {parent=lowerarm_r, loc=<-22.3492 0 0>, rot=<-0.7081 -0.0507 0.0975 -0.6975>}
 index_metacarpal_r {parent=hand_r, loc=<-1.6534 -0.0181 2.9601>, rot=<0.0158 -0.0124 -0.0173 -0.9996>}
 index_01_r {parent=index_metacarpal_r, loc=<-8.3008 -0.0357 0.0524>, rot=<-0.0044 0.0485 0.0924 -0.9945>}
 index_02_r {parent=index_01_r, loc=<-3.3589 0.3003 -0.0453>, rot=<-0.002 0.0027 0.3225 -0.9466>}
 index_03_r {parent=index_02_r, loc=<-2.5578 0.1231 0.0251>, rot=<-0.002 0.0067 0.2558 -0.9667>}
 middle_metacarpal_r {parent=hand_r, loc=<-2.1856 0.1726 0.5984>, rot=<0.006 0.0168 -0.0329 -0.9993>}
 middle_01_r {parent=middle_metacarpal_r, loc=<-7.3273 0.2447 -0.0129>, rot=<-0.0208 0.0042 0.0858 -0.9961>}
 middle_02_r {parent=middle_01_r, loc=<-4.0142 -0.1411 -0.07>, rot=<-0.0217 0.0259 0.3512 -0.9357>}
 middle_03_r {parent=middle_02_r, loc=<-2.6074 -0.0534 -0.1312>, rot=<0.0083 0.0188 0.1708 -0.9851>}
 pinky_metacarpal_r {parent=hand_r, loc=<-1.7617 0.9753 -3.3814>, rot=<-0.0397 0.0814 -0.069 -0.9935>}
 pinky_01_r {parent=pinky_metacarpal_r, loc=<-7.1707 0.4832 0.1094>, rot=<-0.0654 -0.0035 0.2242 -0.9723>}
 pinky_02_r {parent=pinky_01_r, loc=<-2.8274 -0.0892 0.1043>, rot=<0.0037 0.0011 0.2854 -0.9584>}
 pinky_03_r {parent=pinky_02_r, loc=<-2.2045 -0.031 0.0078>, rot=<-0.0002 0.0098 0.1445 -0.9895>}
 ring_metacarpal_r {parent=hand_r, loc=<-2.2073 0.3627 -1.3638>, rot=<-0.0284 0.0722 -0.0458 -0.9959>}
 ring_01_r {parent=ring_metacarpal_r, loc=<-7.0297 0.6197 0.1446>, rot=<-0.0489 -0.0252 0.1892 -0.9804>}
 ring_02_r {parent=ring_01_r, loc=<-3.6171 -0.4648 0.1027>, rot=<0 0 0.2806 -0.9598>}
 ring_03_r {parent=ring_02_r, loc=<-2.4153 -0.1141 -0.0199>, rot=<0 0 0.1189 -0.9929>}
 thumb_01_r {parent=hand_r, loc=<-1.3119 0.3192 2.459>, rot=<0.6546 -0.4042 0.0664 -0.6354>}
 thumb_02_r {parent=thumb_01_r, loc=<-5.9335 0.0447 -0.0401>, rot=<-0.0037 0.0213 0.298 -0.9543>}
 thumb_03_r {parent=thumb_02_r, loc=<-2.2223 -0.3682 0.0907>, rot=<0.0118 -0.013 0.0066 -0.9998>}
 weapon_r {parent=hand_r, loc=<2.9556 6.0388 -9.2061>, rot=<0.0016 -0.0067 -0.1654 -0.9862>}
 lowerarm_twist_01_r {parent=lowerarm_r, loc=<-17.855 0 0>, rot=<0 0 0 1>}
 lowerarm_twist_02_r {parent=lowerarm_r, loc=<-9.6199 0 0>, rot=<0 0 0 1>}
 upperarm_twist_01_r {parent=upperarm_r, loc=<-0.0041 0 0>, rot=<0 0 0 1>}
 upperarm_twist_02_r {parent=upperarm_r, loc=<-13.7942 0 0>, rot=<0 0 0 1>}
 dyn_armband_knot_MSTR {parent=upperarm_r, loc=<-15.5423 7.3517 2.1108>, rot=<-0.6295 -0.6523 -0.3869 -0.169>}
 dyn_armband_knot {parent=dyn_armband_knot_MSTR, loc=<0.4877 0 0>, rot=<0 0 0 1>}
 dyn_armband_low_01 {parent=dyn_armband_knot, loc=<2.0963 -0.259 -1.4509>, rot=<-0.0037 -0.1522 0.1128 -0.9819>}
 dyn_armband_low_02 {parent=dyn_armband_low_01, loc=<6.7826 0 0>, rot=<0 0 0 1>}
 dyn_armband_hi_01 {parent=dyn_armband_knot, loc=<1.8449 0.5097 0.7988>, rot=<0.0115 0.1786 0.0321 -0.9833>}
 dyn_armband_hi_02 {parent=dyn_armband_hi_01, loc=<5.0757 0 0>, rot=<0 0 0 1>}
 dyn_strap_clip_fr_r_MSTR {parent=clavicle_r, loc=<-12.1796 -11.1367 2.2177>, rot=<-0.3818 0.6519 -0.4186 -0.504>}
 dyn_strap_clip_fr_r {parent=dyn_strap_clip_fr_r_MSTR, loc=<0.2618 0.0439 0.2724>, rot=<0.0223 0.2541 0.0075 -0.9669>}
 dyn_strap_clip_bk_r_MSTR {parent=clavicle_r, loc=<-11.6531 11.3141 0.439>, rot=<-0.4921 -0.6038 -0.4008 0.4822>}
 dyn_strap_clip_bk_r {parent=dyn_strap_clip_bk_r_MSTR, loc=<0.1948 0.0816 0.3487>, rot=<0.024 0.1358 0.0033 -0.9904>}
 base_backstraps_r {parent=clavicle_r, loc=<-10.5973 0 -8.0766>, rot=<0 -0.0993 0 -0.9951>}
 dyn_backstraps_r {parent=base_backstraps_r, loc=<0 0 0>, rot=<0 0 0 1>}
 Base_clavPushJnt_r {parent=clavicle_r, loc=<-6.5322 -1.9955 -7.6349>, rot=<0 -0.0993 0 -0.9951>}
 dyn_clavPushJnt_r {parent=Base_clavPushJnt_r, loc=<0 0 0>, rot=<0 0 0 1>}
 neck_01 {parent=spine_05, loc=<11.3138 0.1529 0>, rot=<0 0 -0.2003 -0.9797>}
 neck_02 {parent=neck_01, loc=<4.5627 0 0>, rot=<0 0 0.0504 -0.9987>}
 head {parent=neck_02, loc=<5.3914 0 0>, rot=<0 0 0.115 -0.9934>}
 faceAttach {parent=head, loc=<-0.8266 -0.1832 0>, rot=<0.0044 -0.3827 -0.0044 -0.9239>, scale=<0.89 0.89 0.89>}
 C_brow_mid {parent=faceAttach, loc=<4.7209 11.2084 4.8303>, rot=<-0.0586 -0.317 0.0114 -0.9466>}
 L_brow_mid {parent=faceAttach, loc=<7.639 10.371 3.157>, rot=<0.0603 -0.3852 0.2489 -0.8866>}
 L_brow_outer {parent=faceAttach, loc=<9.1042 8.1218 1.4857>, rot=<0.2349 -0.2525 0.3549 -0.869>}
 L_cheek_inner {parent=faceAttach, loc=<4.2785 8.761 -1.4435>, rot=<0.0454 -0.6262 0.3191 -0.7099>}
 C_lip_upper_mid {parent=faceAttach, loc=<-0.3154 10.7989 -0.2464>, rot=<0.2097 -0.3713 -0.0897 -0.9001>}
 L_lip_upper_outer {parent=faceAttach, loc=<0.8467 10.1076 -1.4454>, rot=<0.1857 -0.3758 0.3015 -0.8564>}
 L_lip_corner {parent=faceAttach, loc=<1.613 9.1457 -2.3124>, rot=<0.0012 -0.375 -0.0189 -0.9269>}
 C_nose_bridge {parent=faceAttach, loc=<2.3602 13.1515 2.4853>, rot=<-0.2002 -0.3676 0.0825 -0.9044>}
 R_brow_mid {parent=faceAttach, loc=<3.0903 10.342 7.7132>, rot=<-0.1347 -0.3556 -0.2231 -0.8976>}
 R_brow_outer {parent=faceAttach, loc=<1.412 8.0675 9.1444>, rot=<-0.1251 -0.3913 -0.4176 -0.8104>}
 R_cheek_inner {parent=faceAttach, loc=<-1.4992 8.7305 4.3237>, rot=<-0.1979 -0.0545 -0.2598 -0.9436>}
 R_lip_corner {parent=faceAttach, loc=<-2.32 9.1134 1.692>, rot=<0.0127 -0.39 0.008 -0.9207>}
 C_jaw {parent=faceAttach, loc=<0.3123 3.2483 0.3582>, rot=<0.0044 -0.3827 -0.0044 -0.9239>}
 L_lip_lower_outer {parent=C_jaw, loc=<1.6966 6.8705 -0.8455>, rot=<-0.0146 -0.0279 0.3432 -0.9387>}
 C_lip_lower_mid {parent=C_jaw, loc=<0.0184 7.4191 -0.884>, rot=<0 -0.0099 -0.0021 -0.9999>}
 R_lip_lower_outer {parent=C_jaw, loc=<-1.6982 6.9042 -0.8853>, rot=<0.0021 -0.0171 -0.3152 -0.9489>}
 teeth_lower {parent=C_jaw, loc=<0.0169 5.3598 -2.8149>, rot=<0 0 0 1>}
 tongue {parent=C_jaw, loc=<0.0178 1.5245 -1.0398>, rot=<-0.1568 0 0 -0.9876>}
 R_eye {parent=faceAttach, loc=<2.2537 8.4351 6.6885>, rot=<0.3823 -0.0165 0.9223 -0.0548>}
 R_eye_lid_upper_mid {parent=R_eye, loc=<-0.0465 -1.169 0.2722>, rot=<0.0559 -0.0011 0.0197 -0.9982>, scale=<1 1.0001 0.9999>}
 R_eye_lid_lower_mid {parent=R_eye, loc=<-0.0141 -0.9451 -0.7006>, rot=<0.035 -0.0006 0.0197 -0.9992>}
 L_eye {parent=faceAttach, loc=<6.6314 8.4647 2.3108>, rot=<0.3819 0.0253 0.9227 0.046>}
 L_eye_lid_upper_mid {parent=L_eye, loc=<0.0251 -1.1623 0.2388>, rot=<0.0523 0 0 -0.9986>}
 L_eye_lid_lower_mid {parent=L_eye, loc=<0.0223 -0.923 -0.7488>, rot=<0.0329 0 0 -0.9995>}
 teeth_upper {parent=faceAttach, loc=<0.0954 9.3381 0.1585>, rot=<0.0044 -0.3827 -0.0044 -0.9239>}
 R_lip_upper_outer {parent=faceAttach, loc=<-1.497 10.1359 0.962>, rot=<-0.0848 -0.2926 -0.3215 -0.8966>}
 dyn_collar_fr_r_MSTR {parent=spine_05, loc=<2.2225 9.5484 5.7859>, rot=<0.8048 0.4894 0.2488 0.2255>}
 dyn_collar_fr_r_01 {parent=dyn_collar_fr_r_MSTR, loc=<0.8292 0 0>, rot=<0 0 0 1>}
 dyn_collar_fr_r_02 {parent=dyn_collar_fr_r_01, loc=<5.5147 -0.1547 0.7264>, rot=<0 0 0 1>}
 dyn_collar_back_MSTR {parent=spine_05, loc=<13.994 -5.0394 0>, rot=<-0.7032 0.0747 0.0747 -0.7032>}
 dyn_collar_back {parent=dyn_collar_back_MSTR, loc=<0.5899 0 0>, rot=<0 0 0 1>}
 dyn_collar_bk_l_MSTR {parent=spine_05, loc=<13.8632 -3.2125 -6.9159>, rot=<0.9052 0.019 -0.1727 0.3878>}
 dyn_collar_bk_l {parent=dyn_collar_bk_l_MSTR, loc=<0.5899 0 0>, rot=<0 0 0 1>}
 dyn_collar_l_MSTR {parent=spine_05, loc=<10.7339 3.1427 -9.8401>, rot=<0.9457 0.1944 -0.258 0.0354>}
 dyn_collar_l {parent=dyn_collar_l_MSTR, loc=<0.8292 0 0>, rot=<0 0 0 1>}
 dyn_collar_fr_l_MSTR {parent=spine_05, loc=<4.2984 8.9751 -8.9508>, rot=<0.9235 0.3447 -0.1532 -0.07>}
 dyn_collar_fr_l_01 {parent=dyn_collar_fr_l_MSTR, loc=<0.8292 0 0>, rot=<0 0 0 1>}
 dyn_collar_fr_l_02 {parent=dyn_collar_fr_l_01, loc=<5.663 -0.4924 0.7959>, rot=<-0.0479 0.1389 0.5158 -0.844>}
 dyn_collar_fr_l_03 {parent=dyn_collar_fr_l_02, loc=<2.6506 0 0>, rot=<0 0 0 1>}
 dyn_backpack {parent=spine_05, loc=<0.9371 -11.6559 0>, rot=<0 0 0 1>}
 dyn_collar_bk_r_MSTR {parent=spine_05, loc=<13.863 -3.2125 6.9159>, rot=<-0.3878 0.1727 -0.019 -0.9052>}
 dyn_collar_bk_r {parent=dyn_collar_bk_r_MSTR, loc=<0.5901 0 -0.0001>, rot=<0 0 0 1>}
 dyn_collar_r_MSTR {parent=spine_05, loc=<10.7337 3.1427 9.8401>, rot=<-0.0354 0.258 -0.1944 -0.9457>}
 dyn_collar_r {parent=dyn_collar_r_MSTR, loc=<0.829 0.0001 0.0001>, rot=<0 0 0 1>}
 attach_cape {parent=spine_05, loc=<9.017 -11.714 0>, rot=<0.0616 -0.7044 -0.0616 -0.7044>}
 cape {parent=attach_cape, type=Null, loc=<0.7198 -1.6365 -0.9846>, rot=<0.0872 0 0 0.9962>}
 attach_backpack {parent=spine_05, loc=<-0.123 -13.473 0>, rot=<-0.0093 -0.707 0.0093 -0.707>}
 Backpack_BR {parent=attach_backpack, type=Null, loc=<0 1.8442 1.0121>, rot=<-0.0131 0 0 0.9999>}
 clavicle_l_null {parent=spine_05, type=Null, loc=<4.3575 -0.1466 -2.3223>, rot=<0.0465 -0.7289 -0.009 -0.6829>}
 clavicle_r_null {parent=spine_05, type=Null, loc=<4.3575 -0.1466 2.3223>, rot=<0.7289 0.0465 -0.6829 0.009>}
 ref_shulderTopPoint_l_base {parent=spine_05, loc=<9.4789 1.491 -17.433>, rot=<0.7605 0.1064 0.6395 0.0376>}
 ref_shulderTopPoint_l {parent=ref_shulderTopPoint_l_base, loc=<0 0 0>, rot=<0 0 0 1>}
 ref_shulderTopPoint_l_ctrl {controlType=EULER_TRANSFORM, parent=ref_shulderTopPoint_l, type=Control, loc=<0 0 0>, rot=<0 0 0 1>, scale=<3 3 3>}
 ref_shulderTopPoint_r_base {parent=spine_05, loc=<9.479 1.491 17.433>, rot=<-0.0376 -0.6395 -0.1064 -0.7605>}
 ref_shulderTopPoint_r {parent=ref_shulderTopPoint_r_base, loc=<0 0 0>, rot=<0 0 0 1>}
 ref_shulderTopPoint_r_ctrl {controlType=EULER_TRANSFORM, parent=ref_shulderTopPoint_r, type=Control, loc=<0 0 0>, rot=<0 0 0 1>, scale=<3 3 3>}
 dyn_chest_zipper_r_MSTR {parent=spine_04, loc=<-0.1718 9.6375 11.4356>, rot=<-0.0288 0.3813 0.9239 -0.012>}
 dyn_chest_zipper_r {parent=dyn_chest_zipper_r_MSTR, loc=<0 0 0.7154>, rot=<0 0 0 1>}
 dyn_collar_zipper_MSTR {parent=spine_04, loc=<2.7589 14.5377 -4.3214>, rot=<-0.0276 -0.8063 -0.5838 -0.0911>}
 dyn_collar_zipper {parent=dyn_collar_zipper_MSTR, loc=<0 0 0.5712>, rot=<0 0 0 1>}
 thigh_l {parent=pelvis, loc=<-2.5693 -0.0109 -7.7501>, rot=<0.0012 0.0572 0.0155 -0.9982>}
 calf_l {parent=thigh_l, loc=<-39.0484 0 0>, rot=<0.0006 -0.0161 -0.0391 -0.9991>}
 calf_twist_01_l {parent=calf_l, loc=<-35.813 -0.0967 0.0397>, rot=<0.0237 0.0095 -0.0035 -0.9997>}
 calf_twist_02_l {parent=calf_l, loc=<-18.8064 -0.0967 0.0397>, rot=<0.0237 0.0095 -0.0035 -0.9997>}
 dyn_boot_zipper_out_l_MSTR {parent=calf_twist_02_l, loc=<-7.0145 4.2622 -4.8369>, rot=<-0.0867 -0.9366 -0.335 -0.0554>}
 dyn_boot_zipper_out_l {parent=dyn_boot_zipper_out_l_MSTR, loc=<0.8318 0 0>, rot=<0 0 0 1>}
 dyn_boot_zipper_in_l_MSTR {parent=calf_twist_02_l, loc=<-6.5767 3.9294 2.8031>, rot=<0.2683 0.4325 0.8556 0.094>}
 dyn_boot_zipper_in_l {parent=dyn_boot_zipper_in_l_MSTR, loc=<0.8318 0 0>, rot=<0 0 0 1>}
 foot_l {parent=calf_l, loc=<-42.561 0.1077 0.4292>, rot=<0.0418 -0.0455 0.0343 -0.9975>}
 ball_l {parent=foot_l, loc=<-6.1599 13.8264 0.6806>, rot=<0.0001 -0.0001 -0.7063 -0.7079>}
 dyn_kneepad_l {parent=calf_l, loc=<0.2806 1.6311 0.1615>, rot=<0 0 0 1>}
 thigh_twist_01_l {parent=thigh_l, loc=<-21.6303 0 0>, rot=<0.0013 0 0 -1>}
 dyn_hip_pouch_fr_l {parent=thigh_twist_01_l, loc=<8.7863 6.351 -5.7183>, rot=<-0.0089 -0.8948 -0.4453 -0.0303>}
 thigh_r {parent=pelvis, loc=<-2.5691 -0.0109 7.7501>, rot=<0.0572 -0.0012 0.9982 0.0155>}
 calf_r {parent=thigh_r, loc=<39.0487 0 0>, rot=<0.0006 -0.0161 -0.0391 -0.9991>}
 calf_twist_01_r {parent=calf_r, loc=<35.8132 0.0967 -0.0397>, rot=<0.0237 0.0095 -0.0035 -0.9997>}
 calf_twist_02_r {parent=calf_r, loc=<18.8064 0.0967 -0.0397>, rot=<0.0237 0.0095 -0.0035 -0.9997>}
 dyn_boot_zipper_in_r_MSTR {parent=calf_twist_02_r, loc=<6.5766 -3.9294 -2.8031>, rot=<0.8556 -0.094 -0.2683 0.4325>}
 dyn_boot_zipper_in_r {parent=dyn_boot_zipper_in_r_MSTR, loc=<0.8318 0 0>, rot=<0 0 0 1>}
 dyn_boot_zipper_out_r_MSTR {parent=calf_twist_02_r, loc=<7.0145 -4.2622 4.8369>, rot=<-0.335 0.0554 0.0867 -0.9366>}
 dyn_boot_zipper_out_r {parent=dyn_boot_zipper_out_r_MSTR, loc=<0.8317 0 0>, rot=<0 0 0 1>}
 foot_r {parent=calf_r, loc=<42.5612 -0.1077 -0.4292>, rot=<0.0418 -0.0455 0.0343 -0.9975>}
 ball_r {parent=foot_r, loc=<6.1598 -13.8264 -0.6806>, rot=<0.0001 -0.0001 -0.7063 -0.7079>}
 dyn_kneepad_r {parent=calf_r, loc=<-0.2806 -1.6311 -0.1615>, rot=<0 -1 0 0>}
 thigh_twist_01_r {parent=thigh_r, loc=<21.6303 0 0>, rot=<0.0013 0 0 -1>}
 dyn_hip_pouch_fr_r {parent=thigh_twist_01_r, loc=<-13.7598 -4.6583 8.1361>, rot=<-0.2171 -0.0194 -0.0237 -0.9757>}
 dyn_hip_pouch_bk_r {parent=thigh_twist_01_r, loc=<-14.2207 1.6071 8.341>, rot=<0.0608 -0.0028 -0.0452 -0.9971>}
 dyn_belt_back_pouch {parent=pelvis, loc=<13.4281 -11.1495 8.3453>, rot=<0.0294 -0.4704 0.88 -0.058>}
 dyn_belt_bullet_01 {parent=pelvis, loc=<12.0068 3.0615 14.0419>, rot=<-0.0181 0.2568 0.9643 -0.0625>}
 dyn_belt_bullet_02 {parent=pelvis, loc=<11.5551 6.5523 12.2223>, rot=<-0.0208 0.2977 0.9525 -0.0616>}
 dyn_belt_bullet_03 {parent=pelvis, loc=<11.162 9.6073 9.8765>, rot=<-0.023 0.3317 0.9411 -0.0609>}
 pelvis_l {parent=pelvis, loc=<0.1622 0 -0.1>, rot=<0 0 0 1>}
 pelvis_r {parent=pelvis, loc=<0.1622 0 0.1>, rot=<0 0 0 1>}
 pelvis_side_l {parent=pelvis, loc=<0.024 0 -0.1>, rot=<0 0 0 1>}
 pelvis_mid_l {parent=pelvis, loc=<0.1163 0 -0.1>, rot=<0 0 0 1>}
 pelvis_side_r {parent=pelvis, loc=<0.024 0 0.1>, rot=<0 0 0 1>}
 pelvis_mid_r {parent=pelvis, loc=<0.1163 0 0.1>, rot=<0 0 0 1>}
 ik_foot_root {parent=root, loc=<0 0 0>, rot=<0 0 0 1>}
 ik_foot_l {parent=ik_foot_root, loc=<15.2608 -4.4176 9.4433>, rot=<0.0357 0.7045 0.0284 -0.7082>}
 ik_foot_r {parent=ik_foot_root, loc=<-15.2609 -4.4176 9.4431>, rot=<0.7082 0.0284 -0.7045 0.0357>}
 ik_hand_root {parent=root, loc=<0 0 0>, rot=<0 0 0 1>}
 ik_hand_gun {parent=ik_hand_root, loc=<-51.9729 4.0884 102.6935>, rot=<0.6084 0.1793 0.4613 -0.6205>}
 ik_hand_l {parent=ik_hand_gun, loc=<53.0286 82.1782 35.2051>, rot=<-0.5102 -0.7906 -0.3387 0>}
 ik_hand_r {parent=ik_hand_gun, loc=<0 0 0>, rot=<0 0 0 1>}
 upperarm_twist_01_l_null {type=Null, loc=<17.7819 -4.8445 141>, rot=<-0.0134 0.3555 -0.014 0.9345>}
]
variables [
 DyanamicsOnMultiplier {type=double, default=0}
 RigidBodiesEnabled {type=bool, default=False}
 apologyEmote {type=bool, default=False}
 isCroching {type=bool, default=False}
 isEmoting {type=bool, default=False}
 isTargetting {type=bool, default=False}
]
