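# Physics asset for the dynamic collar of the medium male Rebirth Soldier player body, in its no-hoodie variant.
# Box bodies form collar panels constrained to small MSTR capsule anchors, with short chains at the front left and right.
# Capsules on head, neck, shoulders and upper arms provide body volumes around the collar.
# Angular limits differ per panel and side, while head-to-neck joints swing and twist freely.
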PhysicsAsset /Game/Characters/Player/Male/Medium/Bodies/M_MED_Rebirth_Soldier/Meshes/M_MED_Rebirth_Soldier_Dyn_Collar_NoHoodie
bodies {
 dyn_collar_back [{shapeType=Box, extentX=11.3518, extentY=20.9498, extentZ=1.2319, center=<4.2656 0.0001 0.5405>, rotation=<rot 9.7663 0 0>}]
 dyn_collar_back_MSTR [{shapeType=Capsule, length=0.5, radius=0.5, rotation=<rot 90 0 0>}]
 dyn_collar_bk_l [{shapeType=Box, extentX=8.1704, extentY=14.2832, extentZ=1.2319, center=<2.9371 0.0149 0.1896>, rotation=<rot 0.3446 -0.1808 -6.3727>}]
 dyn_collar_bk_l_MSTR [{shapeType=Capsule, length=0.5, radius=0.5, rotation=<rot 90 0 0>}]
 dyn_collar_bk_r [{shapeType=Box, extentX=8.9465, extentY=14.5753, extentZ=1.2116, center=<1.9638 -1.0321 -0.2547>, rotation=<rot -4.0342 4.7187 -2.616>}]
 dyn_collar_bk_r_MSTR [{shapeType=Capsule, length=0.5, radius=0.5, rotation=<rot 90 0 0>}]
 dyn_collar_fr_l_01 [{shapeType=Box, extentX=8.2121, extentY=10.1463, extentZ=0.9701, center=<3.3426 0.0001 0.3763>, rotation=<rot 10 0 0>}]
 dyn_collar_fr_l_02 [{shapeType=Box, extentX=2.5458, extentY=2.1682, extentZ=0.8797, center=<1.3206 0.0001 0.0199>, rotation=<rot 10 0 0>}]
 dyn_collar_fr_l_03 [{shapeType=Box, extentX=2.5458, extentY=2.1682, extentZ=0.8797, center=<1.3206 0.0001 0.0199>, rotation=<rot 10 0 0>}]
 dyn_collar_fr_l_MSTR [{shapeType=Capsule, length=0.5, radius=0.5, rotation=<rot 90 0 0>}]
 dyn_collar_fr_r_01 [{shapeType=Box, extentX=8.8052, extentY=19.0676, extentZ=1.2881, center=<2.5082 0.6561 0.8239>}]
 dyn_collar_fr_r_02 [{shapeType=Box, extentX=6.7292, extentY=10.7804, extentZ=1.3181, center=<2.8898 1.2511 0.6653>, rotation=<rot 7.069 0 0>}]
 dyn_collar_fr_r_MSTR [{shapeType=Capsule, length=0.5, radius=0.5, rotation=<rot 90 0 0>}]
 dyn_collar_l [{shapeType=Box, extentX=10.5295, extentY=18.3586, extentZ=0.8686, center=<1.5858 -0.1056 -0.3945>, rotation=<rot 7.8316 -3.4624 3.3008>}]
 dyn_collar_l_MSTR [{shapeType=Capsule, length=0.5, radius=0.5, rotation=<rot 90 0 0>}]
 dyn_collar_r [{shapeType=Box, extentX=10.1639, extentY=15.3507, extentZ=0.6471, center=<1.7204 -0.244 -0.9805>, rotation=<rot 0.4876 1.4203 -8.4022>}]
 dyn_collar_r_MSTR [{shapeType=Capsule, length=0.5, radius=0.5, rotation=<rot 90 0 0>}]
 head [
  {
   shapeType=Capsule
   length=4.5178
   radius=4.9569
   shapeName=headTubeA_r
   center=<-0.8114 4.8339 0.3711>
   rotation=<rot -0.6323 20.2489 83.5797>
  }
  {shapeType=Capsule, length=7.2378, radius=7.1108, center=<4.9418 3.0541 0.0001>, rotation=<rot 90 0 0>}
  {shapeType=Capsule, length=5.2477, radius=7.1108, center=<5.0858 1.6577 0.0001>, rotation=<rot 90 -18.5209 0>}
 ]
 neck_01 [{
   shapeType=Capsule
   length=0.7094
   radius=6.4623
   shapeName=headTubeA_r
   center=<0.6667 0.7869 0.1334>
   rotation=<rot -5.4012 -97.4133 93.5332>
  }]
 neck_02 [{
   shapeType=Capsule
   length=0.7993
   radius=6.6246
   shapeName=headTubeA_r
   center=<0.6484 0.5064 0.1058>
   rotation=<rot -5.9305 -87.4176 92.5438>
  }]
 ref_shulderTopPoint_l [{
   shapeType=Capsule
   length=6.0082
   radius=4.3092
   shapeName=ref_shulderTopPoint_l_capsule
   center=<-0.6799 1.529 -3.6877>
   rotation=<rot -4.4486 10.5099 -90.824>
  }]
 ref_shulderTopPoint_r [{
   shapeType=Capsule
   length=6.1477
   radius=4.3092
   shapeName=ref_shulderTopPoint_r_capsule
   center=<-1.0798 -2.061 -3.7375>
   rotation=<rot -4.4347 -11.4283 90.8958>
  }]
 upperarm_l [{
   shapeType=Capsule
   length=21.5559
   radius=7.5053
   shapeName=upperarm_l_capsule
   center=<11.0339 -0.0972 -0.7088>
   rotation=<rot 13.6321 -89.9928 87.2455>
  }]
 upperarm_r [{
   shapeType=Capsule
   length=21.5559
   radius=7.5053
   shapeName=upperarm_r_capsule
   center=<-11.0147 0.9782 0.0824>
   rotation=<rot 13.6321 -89.9928 -92.7545>
  }]
}
constraints [
 {bone1Name=dyn_collar_fr_r_01, bone2Name=dyn_collar_fr_r_MSTR, swing1LimitDegrees=5, swing2LimitDegrees=50, twistLimitDegrees=15}
 {bone1Name=dyn_collar_fr_r_02, bone2Name=dyn_collar_fr_r_01, swing1LimitDegrees=0, swing2LimitDegrees=40, twistLimitDegrees=0}
 {bone1Name=dyn_collar_r, bone2Name=dyn_collar_r_MSTR, swing1LimitDegrees=20, swing2LimitDegrees=20, twistLimitDegrees=10}
 {bone1Name=dyn_collar_bk_r, bone2Name=dyn_collar_bk_r_MSTR, swing1LimitDegrees=75, swing2LimitDegrees=75, twistLimitDegrees=75}
 {bone1Name=dyn_collar_back, bone2Name=dyn_collar_back_MSTR, swing1LimitDegrees=90, swing2LimitDegrees=90, twistLimitDegrees=90}
 {bone1Name=dyn_collar_bk_l, bone2Name=dyn_collar_bk_l_MSTR, swing1LimitDegrees=75, swing2LimitDegrees=75, twistLimitDegrees=75}
 {bone1Name=dyn_collar_l, bone2Name=dyn_collar_l_MSTR, swing1LimitDegrees=75, swing2LimitDegrees=75, twistLimitDegrees=75}
 {bone1Name=dyn_collar_fr_l_01, bone2Name=dyn_collar_fr_l_MSTR, swing1LimitDegrees=75, swing2LimitDegrees=75, twistLimitDegrees=75}
 {bone1Name=dyn_collar_fr_l_02, bone2Name=dyn_collar_fr_l_01, swing1LimitDegrees=15, swing2LimitDegrees=45, twistLimitDegrees=5}
 {bone1Name=dyn_collar_fr_l_03, bone2Name=dyn_collar_fr_l_02, swing1LimitDegrees=15, swing2LimitDegrees=45, twistLimitDegrees=5}
 {
  bone1Name=head
  bone2Name=neck_02
  swing1LimitDegrees=45
  swing1Motion=Free
  swing2LimitDegrees=45
  swing2Motion=Free
  twistLimitDegrees=45
  twistMotion=Free
 }
 {
  bone1Name=neck_02
  bone2Name=neck_01
  swing1LimitDegrees=45
  swing1Motion=Free
  swing2LimitDegrees=45
  swing2Motion=Free
  twistLimitDegrees=45
  twistMotion=Free
 }
]
